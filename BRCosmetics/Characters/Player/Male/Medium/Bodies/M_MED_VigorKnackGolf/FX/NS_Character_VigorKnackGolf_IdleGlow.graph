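# Niagara idle glow for the VigorKnackGolf character body: one CapsuleGlow sprite emitter burst-spawns five groups with per-group size, color and camera offset.
# Sprites follow the skeleton via weighted blends of three socket samples and slerped rotations, with alternate sockets chosen by LOD.
# Secondary color, variant, reactivity, time of day, platform and front-end-aware fade-in drive brightness.
# The last group hides when the head is hidden or via MaskOff.
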NiagaraSystem /BRCosmetics/Characters/Player/Male/Medium/Bodies/M_MED_VigorKnackGolf/FX/NS_Character_VigorKnackGolf_IdleGlow
dependencies {
 usedDataInterfaces [&NiagaraDataInterfaceCurve, &NiagaraDataInterfacePlatformSet, &NiagaraDataInterfaceSocketReader]
 usedDynamicInputs [&ClampFloat, &FNTimeOfDayScalar]
 usedModules [
  &SystemState
  &self:SystemSpawnScript.SystemScriptSource.SystemScriptGraph.NiagaraNodeAssignment_1.SetVariables_#1
  &self:SystemSpawnScript.SystemScriptSource.SystemScriptGraph.NiagaraNodeAssignment_0.SetVariables_#2
  &self:SystemSpawnScript.SystemScriptSource.SystemScriptGraph.NiagaraNodeAssignment_2.SetVariables_#3
 ]
}
emitters {CapsuleGlow {
  emitterUpdateScript [
   &EmitterState(
    Inactive Response=::NewEnumerator1
    Life Cycle Mode=::NewEnumerator1
    Loop Behavior=::NewEnumerator1
    Loop Duration Mode=::NewEnumerator1
    UseLoopDelay=false
   )
   &SpawnBurst_Instantaneous(Spawn Count=4, Use Spawn Probability=false, UseLoopCountLimit=false)
   &SpawnBurst_Instantaneous(Spawn Count=3, Spawn Group=1, Use Spawn Probability=false, UseLoopCountLimit=false) {module=SpawnBurst_Instantaneous001}
   &SpawnBurst_Instantaneous(Spawn Count=4, Spawn Group=2, Use Spawn Probability=false, UseLoopCountLimit=false) {module=SpawnBurst_Instantaneous005}
   &SpawnBurst_Instantaneous(Spawn Group=3, Use Spawn Probability=false, UseLoopCountLimit=false) {module=SpawnBurst_Instantaneous003}
   &SpawnBurst_Instantaneous(Spawn Count=3, Spawn Group=4, Use Spawn Probability=false, UseLoopCountLimit=false) {module=SpawnBurst_Instantaneous004}
   /SetVariables(Emitter.UseAltBone:NiagaraBool=&SetBoolByIntComparison(A=$User.CurrentLOD, B=1, Comparison Type=::NewEnumerator0)) {isSetParameters=true}
  ]
  particleSpawnScript [
   /SetVariables(Particles.SpawnGroup=$Emitter.SpawnGroup) {isSetParameters=true}
   &InitializeParticle(
    Color=&MergeHelper_ColorSwitch(
     0=<rgba 1 1 1 0.12>
     1=&MergeHelper_ColorSwitch(
      0=<rgba 1 1 1 0.05>
      1=<rgba 1 1 1 0.05>
      2=<rgba 1 1 1 0.015>
      3=?
      Index=/CascadeToNiagaraConverter/NiagaraScripts/DynamicInputScripts/SequentialInt(Use Array Data:NiagaraBool=false, First Number:NiagaraInt32=0, Sequence:NiagaraDataInterfaceArrayInt32=?, Last Number:NiagaraInt32=2)
      Num=::NewEnumerator1
     )
     2=&MergeHelper_ColorSwitch(
      0=<rgba 1 1 1 0.025>
      1=<rgba 1 1 1 0.03>
      2=<rgba 1 1 1 0.035>
      3=<rgba 1 1 1 0.035>
      Index=&SequentialInt(Use Array Data:NiagaraBool=false, First Number:NiagaraInt32=0, Sequence:NiagaraDataInterfaceArrayInt32=?, Last Number:NiagaraInt32=3)
      Num=::NewEnumerator2
     )
     3=<rgba 1 1 1 0.1>
     4=<rgba 1 1 1 0.25>
     Index=$Particles.SpawnGroup
     Num=::NewEnumerator3
    )
    Sprite Size=&MergeHelper_Vector2Switch(
     0=<4 7.28>
     1=&MergeHelper_Vector2Switch(
      0=<5.7 7.8>
      1=<5.7 7.8>
      2=<6.65 9.1>
      3=?
      4=?
      Index=&SequentialInt(Use Array Data:NiagaraBool=false, First Number:NiagaraInt32=0, Sequence:NiagaraDataInterfaceArrayInt32=?, Last Number:NiagaraInt32=2)
      Num=::NewEnumerator1
     )
     2=&MergeHelper_Vector2Switch(
      0=<12 16>
      1=<13 16>
      2=<12.58 12.58>
      3=<12.58 12.58>
      4=?
      Index=&SequentialInt(Use Array Data:NiagaraBool=false, First Number:NiagaraInt32=0, Sequence:NiagaraDataInterfaceArrayInt32=?, Last Number:NiagaraInt32=3)
      Num=::NewEnumerator2
     )
     3=<12 18>
     4=&MergeHelper_Vector2Switch(
      0=<7.2 12.15>
      1=<7.2 7.2>
      2=<7.2 7.2>
      3=?
      4=?
      Index=&SequentialInt(Use Array Data:NiagaraBool=false, First Number:NiagaraInt32=0, Sequence:NiagaraDataInterfaceArrayInt32=?, Last Number:NiagaraInt32=2)
      Num=::NewEnumerator1
     )
     Index=$Particles.SpawnGroup
     Num=::NewEnumerator3
    )
    Sprite Size Mode=::NewEnumerator4
    UsePositionOffset=false
    Write Material Random=true
   )
   &CameraOffset(Camera Offset Amount=&MergeHelper_FloatSwitch(
     0=3
     1=5
     2=&MakeCustomFloatFromBool(Bool=$Emitter.UseAltBone, True Float=0, False Float=5)
     3=5
     4=3
     Index=$Emitter.SpawnGroup
     Num=::NewEnumerator3
    ))
  ]
  particleUpdateScript [
   &ParticleState(Kill Particles When Lifetime Has Elapsed=false, Let Infinitely Lived Particles Die When Emitter Deactivates=true)
   /SetVariables(Particles.Offset:Vector3f=&MergeHelper_VectorSwitch(
     0=&FN_FrontEndVector3Modifier(
      Gameplay=&MergeHelper_VectorSwitch(
       0=<0 0 0>
       1=<0 0 0>
       2=<0 -1.5 0>
       3=<0 -1.5 0>
       Index=&SequentialInt(Use Array Data:NiagaraBool=false, First Number:NiagaraInt32=0, Sequence:NiagaraDataInterfaceArrayInt32=?, Last Number:NiagaraInt32=3)
       Num=::NewEnumerator2
      )
      Front End=<0 -1 0>
     )
     1=&MergeHelper_VectorSwitch(
      0=<0 -0.7 0>
      1=<0 -0.7 0>
      2=<0 0 0>
      3=?
      Index=&SequentialInt(Use Array Data:NiagaraBool=false, First Number:NiagaraInt32=0, Sequence:NiagaraDataInterfaceArrayInt32=?, Last Number:NiagaraInt32=2)
      Num=::NewEnumerator1
     )
     2=<0 -2.5 0>
     3=<0 -3 0>
     4=<0 -1.65 0>
     Index=$Particles.SpawnGroup
     Num=::NewEnumerator3
    )) {isSetParameters=true}
   &SocketLocation(
    Orientation Sampling=::NewEnumerator1
    SamplePosOffset=true
    Sampled Position Offset=$Particles.Offset
    Socket Index=&MergeHelper_IntSwitch(
     0=&SequentialInt(Use Array Data:NiagaraBool=false, First Number:NiagaraInt32=0, Sequence:NiagaraDataInterfaceArrayInt32=?, Last Number:NiagaraInt32=3)
     1=&SequentialInt(Use Array Data:NiagaraBool=false, First Number:NiagaraInt32=4, Sequence:NiagaraDataInterfaceArrayInt32=?, Last Number:NiagaraInt32=6)
     2=&SequentialInt(Use Array Data:NiagaraBool=false, First Number:NiagaraInt32=7, Sequence:NiagaraDataInterfaceArrayInt32=?, Last Number:NiagaraInt32=10)
     3=11
     4=&SequentialInt(Use Array Data:NiagaraBool=false, First Number:NiagaraInt32=12, Sequence:NiagaraDataInterfaceArrayInt32=?, Last Number:NiagaraInt32=14)
     5=?
     Index=$Particles.SpawnGroup
     Num=::NewEnumerator3
    )
    Socket Reader=$System.Socket Reader
    Socket Sampling Mode=::NewEnumerator3
    VelocitySpeedLimit=false
   )
   &SocketLocation(
    Orientation Sampling=::NewEnumerator1
    Position Sampling=::NewEnumerator1
    SamplePosOffset=true
    Sampled Position Offset=$Particles.Offset
    Socket Index=&MergeHelper_IntSwitch(
     0=&MergeHelper_IntSwitch(
      0=17
      1=18
      2=&MakeCustomIntFromBool(Bool=$Emitter.UseAltBone, True INT=43, False INT=15)
      3=&MakeCustomIntFromBool(Bool=$Emitter.UseAltBone, True INT=42, False INT=16)
      4=?
      5=?
      Index=&SequentialInt(Use Array Data:NiagaraBool=false, First Number:NiagaraInt32=0, Sequence:NiagaraDataInterfaceArrayInt32=?, Last Number:NiagaraInt32=3)
      Num=::NewEnumerator2
     )
     1=&MergeHelper_IntSwitch(
      0=21
      1=22
      2=37
      3=?
      4=?
      5=?
      Index=&SequentialInt(Use Array Data:NiagaraBool=false, First Number:NiagaraInt32=0, Sequence:NiagaraDataInterfaceArrayInt32=?, Last Number:NiagaraInt32=2)
      Num=::NewEnumerator1
     )
     2=&MergeHelper_IntSwitch(
      0=33
      1=31
      2=&MakeCustomIntFromBool(Bool=$Emitter.UseAltBone, True INT=38, False INT=20)
      3=&MakeCustomIntFromBool(Bool=$Emitter.UseAltBone, True INT=39, False INT=19)
      4=?
      5=?
      Index=&SequentialInt(Use Array Data:NiagaraBool=false, First Number:NiagaraInt32=0, Sequence:NiagaraDataInterfaceArrayInt32=?, Last Number:NiagaraInt32=3)
      Num=::NewEnumerator2
     )
     3=35
     4=0
     5=?
     Index=$Particles.SpawnGroup
     Num=::NewEnumerator3
    )
    Socket Reader=$System.Socket Reader
    Socket Sampling Mode=::NewEnumerator3
    VelocitySpeedLimit=false
   ) {module=SocketLocation001}
   &SocketLocation(
    Orientation Sampling=::NewEnumerator1
    Position Sampling=::NewEnumerator1
    SamplePosOffset=true
    Sampled Position Offset=$Particles.Offset
    Socket Index=&MergeHelper_IntSwitch(
     0=&MergeHelper_IntSwitch(
      0=27
      1=28
      2=26
      3=25
      4=?
      5=?
      Index=&SequentialInt(Use Array Data:NiagaraBool=false, First Number:NiagaraInt32=0, Sequence:NiagaraDataInterfaceArrayInt32=?, Last Number:NiagaraInt32=5)
      Num=::NewEnumerator2
     )
     1=&MergeHelper_IntSwitch(
      0=29
      1=30
      2=0
      3=?
      4=?
      5=?
      Index=&SequentialInt(Use Array Data:NiagaraBool=false, First Number:NiagaraInt32=0, Sequence:NiagaraDataInterfaceArrayInt32=?, Last Number:NiagaraInt32=2)
      Num=::NewEnumerator1
     )
     2=&MergeHelper_IntSwitch(
      0=34
      1=32
      2=&MakeCustomIntFromBool(Bool=$Emitter.UseAltBone, True INT=40, False INT=23)
      3=&MakeCustomIntFromBool(Bool=$Emitter.UseAltBone, True INT=41, False INT=24)
      4=?
      5=?
      Index=&SequentialInt(Use Array Data:NiagaraBool=false, First Number:NiagaraInt32=0, Sequence:NiagaraDataInterfaceArrayInt32=?, Last Number:NiagaraInt32=3)
      Num=::NewEnumerator2
     )
     3=36
     4=0
     5=?
     Index=$Particles.SpawnGroup
     Num=::NewEnumerator3
    )
    Socket Reader=$System.Socket Reader
    Socket Sampling Mode=::NewEnumerator3
    VelocitySpeedLimit=false
   ) {module=SocketLocation002}
   /SetVariables(Particles.Position=&MergeHelper_VectorSwitch(
     0=&MergeHelper_VectorSwitch(
      0=&self:3Lerp(
       PosA:NiagaraPosition=$Output.SocketLocation.SampledPosition
       PosB:NiagaraPosition=$Output.SocketLocation001.SampledPosition
       PosC:NiagaraPosition=$Output.SocketLocation002.SampledPosition
       WeightA:NiagaraFloat=0.413
       WeightB:NiagaraFloat=0.344
       WeightC:NiagaraFloat=0.243
      )
      1=&self:3Lerp(
       PosA:NiagaraPosition=$Output.SocketLocation.SampledPosition
       PosB:NiagaraPosition=$Output.SocketLocation001.SampledPosition
       PosC:NiagaraPosition=$Output.SocketLocation002.SampledPosition
       WeightA:NiagaraFloat=0.413
       WeightB:NiagaraFloat=0.344
       WeightC:NiagaraFloat=0.243
      )
      2=&self:3Lerp(
       PosA:NiagaraPosition=$Output.SocketLocation.SampledPosition
       PosB:NiagaraPosition=$Output.SocketLocation001.SampledPosition
       PosC:NiagaraPosition=$Output.SocketLocation002.SampledPosition
       WeightA:NiagaraFloat=0.357
       WeightB:NiagaraFloat=0.348
       WeightC:NiagaraFloat=0.295
      )
      3=&self:3Lerp(
       PosA:NiagaraPosition=$Output.SocketLocation.SampledPosition
       PosB:NiagaraPosition=$Output.SocketLocation001.SampledPosition
       PosC:NiagaraPosition=$Output.SocketLocation002.SampledPosition
       WeightA:NiagaraFloat=0.357
       WeightB:NiagaraFloat=0.348
       WeightC:NiagaraFloat=0.295
      )
      4=&self:3Lerp(
       PosA:NiagaraPosition=$Output.SocketLocation.SampledPosition
       PosB:NiagaraPosition=$Output.SocketLocation001.SampledPosition
       PosC:NiagaraPosition=$Output.SocketLocation002.SampledPosition
       WeightA:NiagaraFloat=0.357
       WeightB:NiagaraFloat=0.348
       WeightC:NiagaraFloat=0.295
      )
      5=&self:3Lerp(
       PosA:NiagaraPosition=$Output.SocketLocation.SampledPosition
       PosB:NiagaraPosition=$Output.SocketLocation001.SampledPosition
       PosC:NiagaraPosition=$Output.SocketLocation002.SampledPosition
       WeightA:NiagaraFloat=0.357
       WeightB:NiagaraFloat=0.348
       WeightC:NiagaraFloat=0.295
      )
      Index=&SequentialInt(Use Array Data:NiagaraBool=false, First Number:NiagaraInt32=0, Sequence:NiagaraDataInterfaceArrayInt32=?, Last Number:NiagaraInt32=3)
      Num=::NewEnumerator2
     )
     1=&MergeHelper_VectorSwitch(
      0=&self:3Lerp(
       PosA:NiagaraPosition=$Output.SocketLocation.SampledPosition
       PosB:NiagaraPosition=$Output.SocketLocation001.SampledPosition
       PosC:NiagaraPosition=$Output.SocketLocation002.SampledPosition
       WeightA:NiagaraFloat=0.565
       WeightB:NiagaraFloat=0.348
       WeightC:NiagaraFloat=0.087
      )
      1=&self:3Lerp(
       PosA:NiagaraPosition=$Output.SocketLocation.SampledPosition
       PosB:NiagaraPosition=$Output.SocketLocation001.SampledPosition
       PosC:NiagaraPosition=$Output.SocketLocation002.SampledPosition
       WeightA:NiagaraFloat=0.565
       WeightB:NiagaraFloat=0.348
       WeightC:NiagaraFloat=0.087
      )
      2=&MergeHelper_VectorSwitch(
       0=Linked() {linkedVariable {name=Output.SocketLocation.SampledPosition, type=NiagaraPosition}}
       1=Linked() {linkedVariable {name=Output.SocketLocation001.SampledPosition, type=NiagaraPosition}}
       2=?
       3=?
       Index=&MakeCustomIntFromBool(Bool=$Emitter.UseAltBone, True INT=1, False INT=0)
       Num=::NewEnumerator0
      )
      3=?
      Index=&SequentialInt(Use Array Data:NiagaraBool=false, First Number:NiagaraInt32=0, Sequence:NiagaraDataInterfaceArrayInt32=?, Last Number:NiagaraInt32=2)
      Num=::NewEnumerator1
     )
     2=&MergeHelper_VectorSwitch(
      0=&self:3Lerp(
       PosA:NiagaraPosition=$Output.SocketLocation.SampledPosition
       PosB:NiagaraPosition=$Output.SocketLocation001.SampledPosition
       PosC:NiagaraPosition=$Output.SocketLocation002.SampledPosition
       WeightA:NiagaraFloat=0.574
       WeightB:NiagaraFloat=0.31
       WeightC:NiagaraFloat=0.117
      )
      1=&self:3Lerp(
       PosA:NiagaraPosition=$Output.SocketLocation.SampledPosition
       PosB:NiagaraPosition=$Output.SocketLocation001.SampledPosition
       PosC:NiagaraPosition=$Output.SocketLocation002.SampledPosition
       WeightA:NiagaraFloat=0.575
       WeightB:NiagaraFloat=0.312
       WeightC:NiagaraFloat=0.113
      )
      2=&MergeHelper_VectorSwitch(
       0=&self:3Lerp(
        PosA:NiagaraPosition=$Output.SocketLocation.SampledPosition
        PosB:NiagaraPosition=$Output.SocketLocation001.SampledPosition
        PosC:NiagaraPosition=$Output.SocketLocation002.SampledPosition
        WeightA:NiagaraFloat=0.571
        WeightB:NiagaraFloat=0.217
        WeightC:NiagaraFloat=0.212
       )
       1=&Lerp_Vector(
        A=Linked() {linkedVariable {name=Output.SocketLocation002.SampledPosition, type=NiagaraPosition}}
        B=Linked() {linkedVariable {name=Output.SocketLocation001.SampledPosition, type=NiagaraPosition}}
        Alpha=0.6081
       )
       2=?
       3=?
       Index=&MakeCustomIntFromBool(Bool=$Emitter.UseAltBone, True INT=1, False INT=0)
       Num=::NewEnumerator0
      )
      3=&MergeHelper_VectorSwitch(
       0=&self:3Lerp(
        PosA:NiagaraPosition=$Output.SocketLocation.SampledPosition
        PosB:NiagaraPosition=$Output.SocketLocation001.SampledPosition
        PosC:NiagaraPosition=$Output.SocketLocation002.SampledPosition
        WeightA:NiagaraFloat=0.571
        WeightB:NiagaraFloat=0.217
        WeightC:NiagaraFloat=0.212
       )
       1=&Lerp_Vector(
        A=Linked() {linkedVariable {name=Output.SocketLocation002.SampledPosition, type=NiagaraPosition}}
        B=Linked() {linkedVariable {name=Output.SocketLocation001.SampledPosition, type=NiagaraPosition}}
        Alpha=0.6081
       )
       2=?
       3=?
       Index=&MakeCustomIntFromBool(Bool=$Emitter.UseAltBone, True INT=1, False INT=0)
       Num=::NewEnumerator0
      )
      Index=&SequentialInt(Use Array Data:NiagaraBool=false, First Number:NiagaraInt32=0, Sequence:NiagaraDataInterfaceArrayInt32=?, Last Number:NiagaraInt32=3)
      Num=::NewEnumerator2
     )
     3=&self:3Lerp(
      PosA:NiagaraPosition=$Output.SocketLocation.SampledPosition
      PosB:NiagaraPosition=$Output.SocketLocation001.SampledPosition
      PosC:NiagaraPosition=$Output.SocketLocation002.SampledPosition
      WeightA:NiagaraFloat=0.536
      WeightB:NiagaraFloat=0.252
      WeightC:NiagaraFloat=0.212
     )
     4=Linked() {linkedVariable {name=Output.SocketLocation.SampledPosition, type=NiagaraPosition}}
     Index=$Particles.SpawnGroup
     Num=::NewEnumerator3
    )) {isSetParameters=true}
   /SetVariables(Particles.InterpolatedQuaternion:Quat4f=/Game/Effects/Niagara/Modules/MergeEmitters/MergeHelper_QuatSwitch(
     0:Quat4f=&MergeHelper_QuatSwitch(
      0:Quat4f=&LerpQuaternion(
       Quaternion A:Quat4f=$Output.SocketLocation001.SampledRotation
       Quaternion B:Quat4f=$Output.SocketLocation.SampledRotation
       Lerp Factor:NiagaraFloat=0.546
       Enable Spherical Linear Interpolation:NiagaraBool=true
      )
      1:Quat4f=&LerpQuaternion(
       Quaternion A:Quat4f=$Output.SocketLocation001.SampledRotation
       Quaternion B:Quat4f=$Output.SocketLocation.SampledRotation
       Lerp Factor:NiagaraFloat=0.546
       Enable Spherical Linear Interpolation:NiagaraBool=true
      )
      2:Quat4f=&LerpQuaternion(
       Quaternion A:Quat4f=$Output.SocketLocation001.SampledRotation
       Quaternion B:Quat4f=$Output.SocketLocation.SampledRotation
       Lerp Factor:NiagaraFloat=0.506
       Enable Spherical Linear Interpolation:NiagaraBool=true
      )
      3:Quat4f=&LerpQuaternion(
       Quaternion A:Quat4f=$Output.SocketLocation001.SampledRotation
       Quaternion B:Quat4f=$Output.SocketLocation.SampledRotation
       Lerp Factor:NiagaraFloat=0.506
       Enable Spherical Linear Interpolation:NiagaraBool=true
      )
      4:Quat4f=?
      5:Quat4f=?
      6:Quat4f=?
      7:Quat4f=?
      8:Quat4f=?
      9:Quat4f=?
      Index:NiagaraInt32=&SequentialInt(Use Array Data:NiagaraBool=false, First Number:NiagaraInt32=0, Sequence:NiagaraDataInterfaceArrayInt32=?, Last Number:NiagaraInt32=3)
      Num:MergeHelper_SwitcherEnum=::NewEnumerator2
     )
     1:Quat4f=&MergeHelper_QuatSwitch(
      0:Quat4f=&LerpQuaternion(
       Quaternion A:Quat4f=$Output.SocketLocation001.SampledRotation
       Quaternion B:Quat4f=$Output.SocketLocation.SampledRotation
       Lerp Factor:NiagaraFloat=0.619
       Enable Spherical Linear Interpolation:NiagaraBool=true
      )
      1:Quat4f=&LerpQuaternion(
       Quaternion A:Quat4f=$Output.SocketLocation001.SampledRotation
       Quaternion B:Quat4f=$Output.SocketLocation.SampledRotation
       Lerp Factor:NiagaraFloat=0.619
       Enable Spherical Linear Interpolation:NiagaraBool=true
      )
      2:Quat4f=$Output.SocketLocation.SampledRotation
      3:Quat4f=?
      4:Quat4f=?
      5:Quat4f=?
      6:Quat4f=?
      7:Quat4f=?
      8:Quat4f=?
      9:Quat4f=?
      Index:NiagaraInt32=&SequentialInt(Use Array Data:NiagaraBool=false, First Number:NiagaraInt32=0, Sequence:NiagaraDataInterfaceArrayInt32=?, Last Number:NiagaraInt32=2)
      Num:MergeHelper_SwitcherEnum=::NewEnumerator1
     )
     2:Quat4f=&MergeHelper_QuatSwitch(
      0:Quat4f=&LerpQuaternion(
       Quaternion A:Quat4f=$Output.SocketLocation001.SampledRotation
       Quaternion B:Quat4f=$Output.SocketLocation.SampledRotation
       Lerp Factor:NiagaraFloat=0.649
       Enable Spherical Linear Interpolation:NiagaraBool=true
      )
      1:Quat4f=&LerpQuaternion(
       Quaternion A:Quat4f=$Output.SocketLocation001.SampledRotation
       Quaternion B:Quat4f=$Output.SocketLocation.SampledRotation
       Lerp Factor:NiagaraFloat=0.648
       Enable Spherical Linear Interpolation:NiagaraBool=true
      )
      2:Quat4f=&LerpQuaternion(
       Quaternion A:Quat4f=$Output.SocketLocation001.SampledRotation
       Quaternion B:Quat4f=$Output.SocketLocation.SampledRotation
       Lerp Factor:NiagaraFloat=0.725
       Enable Spherical Linear Interpolation:NiagaraBool=true
      )
      3:Quat4f=&LerpQuaternion(
       Quaternion A:Quat4f=$Output.SocketLocation001.SampledRotation
       Quaternion B:Quat4f=$Output.SocketLocation.SampledRotation
       Lerp Factor:NiagaraFloat=0.725
       Enable Spherical Linear Interpolation:NiagaraBool=true
      )
      4:Quat4f=?
      5:Quat4f=?
      6:Quat4f=?
      7:Quat4f=?
      8:Quat4f=?
      9:Quat4f=?
      Index:NiagaraInt32=&SequentialInt(Use Array Data:NiagaraBool=false, First Number:NiagaraInt32=0, Sequence:NiagaraDataInterfaceArrayInt32=?, Last Number:NiagaraInt32=3)
      Num:MergeHelper_SwitcherEnum=::NewEnumerator2
     )
     3:Quat4f=&LerpQuaternion(
      Quaternion A:Quat4f=$Output.SocketLocation001.SampledRotation
      Quaternion B:Quat4f=$Output.SocketLocation.SampledRotation
      Lerp Factor:NiagaraFloat=0.68
      Enable Spherical Linear Interpolation:NiagaraBool=true
     )
     4:Quat4f=$Output.SocketLocation.SampledRotation
     5:Quat4f=?
     6:Quat4f=?
     7:Quat4f=?
     8:Quat4f=?
     9:Quat4f=?
     Index:NiagaraInt32=$Particles.SpawnGroup
     Num:MergeHelper_SwitcherEnum=::NewEnumerator3
    )) {isSetParameters=true}
   &SpriteFacingAndAlignment(
    Alignment Coordinate Space=::Local
    Facing Coordinate Space=::Local
    Sprite Alignment=&RotateVectorByQuaternion(Vector To Rotate=<0 0 1>, Quaternion=$Particles.InterpolatedQuaternion)
    Sprite Facing=&RotateVectorByQuaternion(Vector To Rotate=<0 1 0>, Quaternion=$Particles.InterpolatedQuaternion)
    WriteAlignment=true
    WriteFacing=true
   )
   &ScaleColor(
    Scale Alpha=&Multiply_Float(
     A=&Multiply_Float(
      A=&MergeHelper_FloatSwitch(
       0=20
       1=20
       2=1
       3=1
       4=1
       5=2
       6=1
       7=2
       8=2
       9=3
       Index=&MakeIntFromFloat(Float=$User.Variant)
       Num=::NewEnumerator8
      )
      B=$System.Reactivity
     )
     B=&Multiply_Float(
      A=&Multiply_Float(
       A=&FN_FrontEndFloatModifier(
        Gameplay=&FloatFromCurve(
         FloatCurve=DataInterface() {propertyValues {Curve {keys [<at 0 0 RCIM_Cubic>,<at 0.5 1 RCIM_Cubic>]}}}
         CurveIndex=$System.Age
         Scale Curve=&FN_PickFloatDependOnPlatform(TargetPlatformActive=1, TargetPlatformNoActive=0.475)
        )
        Front End=&FloatFromCurve(
         FloatCurve=DataInterface() {propertyValues {Curve {keys [<at 0.25 0 RCIM_Cubic>,<at 1 1 RCIM_Cubic>]}}}
         CurveIndex=$System.Age
         Scale Curve=&FN_PickFloatDependOnPlatform(TargetPlatformActive=0.75, TargetPlatformNoActive=0.1)
        )
       )
       B=&MakeCustomFloatFromBool(
        Bool=&Boolean_AndOperation(
         Bool1=&Boolean_OrOperation(
          Bool1=$User.bisHeadHidden
          Bool2=&SetBoolByIntComparison(A=&MakeIntFromFloat(Float=$User.MaskOff), B=1, Comparison Type=::NewEnumerator4)
         )
         Bool2=&SetBoolByIntComparison(A=$Particles.SpawnGroup, B=4, Comparison Type=::NewEnumerator4)
        )
        True Float=0
        False Float=1
       )
      )
      B=$System.TODFloat
     )
    )
    Scale RGB=&MakeVectorFromLinearColorRGB(Color=&Multiply_LinearColorByFloat(
      Linear Color=$User.Secondary Color
      Float=&MergeHelper_FloatSwitch(
       0=50
       1=0.3
       2=5
       3=3
       4=3.1
       5=3
       6=2.3
       7=10
       8=10
       9=20
       Index=&MakeIntFromFloat(Float=$User.Variant)
       Num=::NewEnumerator8
      )
     ))
    ScaleA=true
    ScaleRGB=true
   )
   &ScaleSpriteSize(
    Scale Sprite Size Mode=::NewEnumerator0
    Uniform Scale Factor=&MakeCustomFloatFromBool(
     Bool=&Boolean_AndOperation(
      Bool1=$User.bisHeadHidden
      Bool2=&SetBoolByIntComparison(A=$Particles.SpawnGroup, B=4, Comparison Type=::NewEnumerator4)
     )
     True Float=0
     False Float=1
    )
   )
   &DynamicMaterialParameters(
    Index 0 Param 1=1
    Index 0 Param 2=10
    Index 0 Param 3=1
    Module.Index0Float0Write=true
    Module.Index0Float1Write=true
    Module.Index0Float2Write=true
    Module.Index0Float3Write=false
   )
  ]
  rendererClasses [&NiagaraSpriteRendererProperties]
  renderers [{rendererClass=&NiagaraSpriteRendererProperties}]
 }}
summary {
 systemName=NS_Character_VigorKnackGolf_IdleGlow
 emitters [{emitterName=CapsuleGlow, rendererClasses [&NiagaraSpriteRendererProperties]}]
 userVariables [
  User.bIsFrontEnd {type=NiagaraBool}
  User.bIsFrontEndPreview {type=NiagaraBool}
  User.bisHeadHidden {type=NiagaraBool}
  User.CurrentLOD {type=NiagaraInt32}
  User.DebugReactivity {type=NiagaraBool}
  User.MaskOff {type=NiagaraFloat, defaultValue=0, description=""}
  User.ReactivityIntensity {type=NiagaraFloat, defaultValue=0, description=""}
  User.Secondary Color {type=LinearColor, description="", defaultValue {struct=LinearColor, value=<rgba 1 1 1 1>}}
  User.Variant {type=NiagaraFloat, defaultValue=0, description=""}
 ]
}
